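# Sound cue for the HoneySting bee companion's interact movement foley in CosmeticCompanions.
# A single wave player feeds the bee movement interact wave straight to the output.
# It uses the companion foley attenuation, the creature pets sound class, a boosted volume and zero priority.
# Concurrency is shared with the quadruped footstep setting rather than a bee-specific one.
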SoundCue /CosmeticCompanions/Assets/Other/HoneySting/Audio/Interact/SC_Mimosa_Other_Bee_Movement_Interact
root=WavePlayer_0
nodes [{id=WavePlayer_0, properties {soundWaveAssetPtr=&/Mimosa_Bee_Movement_Interact}}]
properties {
 attenuationSettings=/Game/Sounds/Attenuation/AS_Companion_Foley
 maxDistance=1200
 priority=0
 soundClassObject=&Fort_Creatures_Pets
 volumeMultiplier=1.5
 concurrencySet [/CosmeticCompanions/Assets/Audio_Shared/Foley/Quadraped/CONC_Mimosa_Quadraped_Footstep]
}
